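; BP_FortPawnComponent_VanillaSoda is a pawn component Blueprint for the VanillaSoda weapon.
; It declares events for begin play, end play, and the VanillaSoda item being added to or removed from inventory.
; All four events are empty here, with no logic connected to them.
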(event EventBeginPlay)

(event EventEndPlay
  (params (EndPlayReason :type EEndPlayReason%20Enum :default "Destroyed")))

(event EventOnVanillaSodaItemAddedToInventory)

(event EventOnVanillaSodaItemRemovedFromInventory)
